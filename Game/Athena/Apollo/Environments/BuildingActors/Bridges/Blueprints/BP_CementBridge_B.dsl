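; BP_CementBridge_B is a cement bridge building actor Blueprint for the Apollo environment.
; Its construction script calls the parent's script, then overrides the world scale of two pillar components and a gear ring component with fixed values.
; The gear ring scale is non-uniform, stretched further along Z.
; BeginPlay, overlap and Tick events are empty, so the graph adds no runtime behavior.
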(event EventBeginPlay)

(event Collision|EventActorBeginOverlap
  (params (OtherActor :type Actor%20Object%20Reference)))

(event EventTick
  (params (DeltaSeconds :type Float%20%28single-precision%29)))

(fn ConstructionScript
  (|CallParentFunction)
  (Transformation|SetWorldScale3D (Variables|Default|GetCementBridge_Pillar) "1.000000,1.000000,1.000000")
  (Transformation|SetWorldScale3D (Variables|Default|GetCementBridge_Pillar1) "1.300000,1.300000,1.300000")
  (Transformation|SetWorldScale3D (Variables|Default|GetS_Gear_Ring) "5.500000,5.500000,11.000000"))
